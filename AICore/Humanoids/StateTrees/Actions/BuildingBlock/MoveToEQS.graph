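# Humanoid AI StateTree building block that runs an environment query, then moves the agent with a Move To task.
# A RunEQS state fetches a single result before MoveTo, and either step failing fails the tree while a completed move succeeds it.
# Under MoveTo, a bool-gated RerunEQS? branch repeats the query with a delay between runs during movement.
# Otherwise, DoNothing is the fallback child, transitioning to MoveTo on tick.
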StateTree /AICore/Humanoids/StateTrees/Actions/BuildingBlock/MoveToEQS
states [
 Root {
  id=#1
  type=STATE
  children [#2, #3]
  tasks []
  transitions []
 }
 RunEQS {
  id=#2
  type=STATE
  tasks [{
    id=#4
    description=Run Env Query
    properties {
     iD=#5
     instance {
      _structType=/Script/GameplayStateTreeModule.StateTreeRunEnvQueryInstanceData
      queryOwner=null
      queryTemplate=null
      runMode=SingleResult
      queryConfig []
      result {refAccessIndex {value=0}}
     }
     node {
      name=Run Env Query
      _structType=/Script/GameplayStateTreeModule.StateTreeRunEnvQueryTask
      bCanEditConsideredForCompletion=true
      bConsideredForCompletion=true
      bHasTaskCompletionDelegateDispatcher=false
      bTaskEnabled=true
      transitionHandlingPriority=Normal
      bindingsBatch {value=65535}
      executionRuntimeTemplateIndex {value=65535}
      instanceDataHandle {index=65535, source=None, stateHandle {index=65535}}
      instanceTemplateIndex {value=65535}
      outputBindingsBatch {value=65535}
     }
    }
   }]
  transitions [
   {id=#6, trigger=ON_STATE_SUCCEEDED, state {id=#3, name=MoveTo}}
   {id=#7, trigger=ON_STATE_FAILED, state {id=#8, name=Root, linkType=FAILED}}
  ]
 }
 MoveTo {
  id=#3
  type=STATE
  children [#9, #10]
  tasks [{
    id=#11
    description=Move To
    properties {
     iD=#12
     instance {
      _structType=/Script/GameplayStateTreeModule.StateTreeMoveToTaskInstanceData
      aIController=null
      acceptableRadius=5
      bAllowPartialPath=true
      bAllowStrafe=false
      bProjectGoalLocation=true
      bReachTestIncludesAgentRadius=true
      bReachTestIncludesGoalRadius=true
      bRequireNavigableEndLocation=true
      bTrackMovingGoal=true
      destinationMoveTolerance=0
      filterClass=null
      targetActor=null
      destination=<0 0 0>
     }
     node {
      name=Move To
      _structType=/Script/GameplayStateTreeModule.StateTreeMoveToTask
      bCanEditConsideredForCompletion=true
      bConsideredForCompletion=true
      bHasTaskCompletionDelegateDispatcher=false
      bSavedShouldCallTick=false
      bTaskEnabled=true
      transitionHandlingPriority=Normal
      bindingsBatch {value=65535}
      executionRuntimeTemplateIndex {value=65535}
      instanceDataHandle {index=65535, source=None, stateHandle {index=65535}}
      instanceTemplateIndex {value=65535}
      outputBindingsBatch {value=65535}
     }
    }
   }]
  transitions [
   {id=#13, trigger=ON_STATE_SUCCEEDED, state {id=#8, name=Root, linkType=SUCCEEDED}}
   {id=#14, trigger=ON_STATE_FAILED, state {id=#8, name=Root, linkType=FAILED}}
  ]
 }
 RerunEQS? {
  id=#9
  type=STATE
  children [#15, #16]
  enter_conditions [{
    id=#17
    description=Bool Compare
    properties {
     iD=#18
     instance {_structType=/Script/StateTreeModule.StateTreeCompareBoolConditionInstanceData, bLeft=false, bRight=true}
     node {
      name=Bool Compare
      _structType=/Script/StateTreeModule.StateTreeCompareBoolCondition
      bInvert=false
      deltaIndent=0
      evaluationMode=Evaluated
      operand=And
      bindingsBatch {value=65535}
      executionRuntimeTemplateIndex {value=65535}
      instanceDataHandle {index=65535, source=None, stateHandle {index=65535}}
      instanceTemplateIndex {value=65535}
      outputBindingsBatch {value=65535}
     }
    }
   }]
  tasks []
  transitions []
 }
 DoNothing {
  id=#10
  type=STATE
  tasks []
  transitions [{id=#19, trigger=ON_TICK, state {id=#3, name=MoveTo}}]
 }
 RunEQS {
  id=#15
  type=STATE
  tasks [{
    id=#20
    description=Run Env Query
    properties {
     iD=#21
     instance {
      _structType=&StateTreeRunEnvQueryInstanceData
      queryOwner=null
      queryTemplate=null
      runMode=SingleResult
      queryConfig []
      result {refAccessIndex {value=0}}
     }
     node {
      name=Run Env Query
      _structType=&StateTreeRunEnvQueryTask
      bCanEditConsideredForCompletion=true
      bConsideredForCompletion=true
      bHasTaskCompletionDelegateDispatcher=false
      bTaskEnabled=true
      transitionHandlingPriority=Normal
      bindingsBatch {value=65535}
      executionRuntimeTemplateIndex {value=65535}
      instanceDataHandle {index=65535, source=None, stateHandle {index=65535}}
      instanceTemplateIndex {value=65535}
      outputBindingsBatch {value=65535}
     }
    }
   }]
  transitions [{id=#22, trigger=ON_STATE_COMPLETED, state {id=#16, name=Wait}}]
 }
 Wait {
  id=#16
  type=STATE
  tasks [{
    id=#23
    description=Delay Task
    properties {
     iD=#24
     instance {_structType=&StateTreeDelayTaskInstanceData, duration=1}
     node {name=Delay Task, _structType=&StateTreeDelayTask}
    }
   }]
  transitions [{id=#25, trigger=ON_STATE_COMPLETED, state {id=#15, name=RunEQS}}]
 }
]
