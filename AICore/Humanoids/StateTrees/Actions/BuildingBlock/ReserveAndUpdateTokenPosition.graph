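# Humanoid AI StateTree building block that repeatedly picks and reserves a token position for a bot, with a global task getting the target actor.
# RunEQS runs a single-result environment query alongside a sprint task, then moves to ReservePosition on success.
# ReservePosition reserves a token position, held briefly after exit, alongside a delay, then loops back to refresh it.
# Failure in either state fails the tree.
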StateTree /AICore/Humanoids/StateTrees/Actions/BuildingBlock/ReserveAndUpdateTokenPosition
global_tasks [{
  id=#1
  description=Get Target Actor
  properties {
   iD=#2
   instance {_structType=&FortStateTreeGetTargetActorTaskInstanceData}
   node {name=Get Target Actor, _structType=&FortStateTreeGetTargetActorTask}
  }
 }]
states [
 Root {id=#3, type=STATE, children [#4], tasks [], transitions []}
 Loop {
  id=#4
  type=STATE
  children [#5, #6]
  tasks []
  transitions []
 }
 RunEQS {
  id=#5
  type=STATE
  tasks [
   {
    id=#7
    description=Run Env Query
    properties {
     iD=#8
     instance {
      _structType=/Script/GameplayStateTreeModule.StateTreeRunEnvQueryInstanceData
      queryOwner=null
      queryTemplate=null
      runMode=SingleResult
      queryConfig []
      result {refAccessIndex {value=0}}
     }
     node {
      name=Run Env Query
      _structType=/Script/GameplayStateTreeModule.StateTreeRunEnvQueryTask
      bCanEditConsideredForCompletion=true
      bConsideredForCompletion=true
      bHasTaskCompletionDelegateDispatcher=false
      bTaskEnabled=true
      transitionHandlingPriority=Normal
      bindingsBatch {value=65535}
      executionRuntimeTemplateIndex {value=65535}
      instanceDataHandle {index=65535, source=None, stateHandle {index=65535}}
      instanceTemplateIndex {value=65535}
      outputBindingsBatch {value=65535}
     }
    }
   }
   {
    id=#9
    description=Bot Sprint
    properties {
     iD=#10
     instance {_structType=/Script/FortniteAIServer.FortStateTreeBotSprintTaskInstanceData, bShouldSprint=true, pawn=null}
     node {
      name=None
      _structType=/Script/FortniteAIServer.FortStateTreeBotSprintTask
      bCanEditConsideredForCompletion=true
      bConsideredForCompletion=true
      bHasTaskCompletionDelegateDispatcher=false
      bTaskEnabled=true
      transitionHandlingPriority=Normal
      bindingsBatch {value=65535}
      executionRuntimeTemplateIndex {value=65535}
      instanceDataHandle {index=65535, source=None, stateHandle {index=65535}}
      instanceTemplateIndex {value=65535}
      outputBindingsBatch {value=65535}
     }
    }
   }
  ]
  transitions [
   {id=#11, trigger=ON_STATE_SUCCEEDED, state {id=#6, name=ReservePosition}}
   {id=#12, trigger=ON_STATE_FAILED, state {id=#13, name=Root, linkType=FAILED}}
  ]
 }
 ReservePosition {
  id=#6
  type=STATE
  tasks [
   {
    id=#14
    description=Reserve Token Position
    properties {
     iD=#15
     instance {
      _structType=/Script/FortniteAIServer.FortStateTreeReserveTokenPositionTaskInstanceData
      acceptableDistanceForPositionReservation=1
      bAllowAddingNewPositionToProvider=true
      holdReservationAtExitDuration=2
      pawn=null
      target=null
      position=<0 0 0>
      tokenTagQuery {autoDescription="", tokenStreamVersion=0, userDescription="", queryTokenStream [], tagDictionary []}
     }
     node {
      name=Reserve Token Position
      _structType=/Script/FortniteAIServer.FortStateTreeReserveTokenPositionTask
      bCanEditConsideredForCompletion=true
      bConsideredForCompletion=true
      bHasTaskCompletionDelegateDispatcher=false
      bTaskEnabled=true
      transitionHandlingPriority=Normal
      bindingsBatch {value=65535}
      executionRuntimeTemplateIndex {value=65535}
      instanceDataHandle {index=65535, source=None, stateHandle {index=65535}}
      instanceTemplateIndex {value=65535}
      outputBindingsBatch {value=65535}
     }
    }
   }
   {
    id=#16
    description=Delay Task
    properties {
     iD=#17
     instance {_structType=&StateTreeDelayTaskInstanceData, duration=2}
     node {name=Delay Task, _structType=&StateTreeDelayTask}
    }
   }
  ]
  transitions [
   {id=#18, trigger=ON_STATE_SUCCEEDED, state {id=#4, name=Loop}}
   {id=#19, trigger=ON_STATE_FAILED, state {id=#20, name=Root, linkType=FAILED}}
  ]
 }
]
