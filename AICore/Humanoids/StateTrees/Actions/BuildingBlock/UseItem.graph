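# Reusable StateTree building block that has a humanoid AI bot equip an item and use it through weapon firing.
# EquipItem leads to FireItem, which gets the current weapon and fires with FireSingleAndRun, while a child state waits for the weapon ability to finish, ending the tree as succeeded or failed.
# A nested Mantis-branch wait re-enters FireItem to chain uses, falling back to an empty NotMantis state if it fails.
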StateTree /AICore/Humanoids/StateTrees/Actions/BuildingBlock/UseItem
states [
 Root {
  id=#1
  type=STATE
  children [#2, #3]
  tasks []
  transitions []
 }
 EquipItem {
  id=#2
  type=STATE
  tasks [{
    id=#4
    description=Equip Item
    properties {
     iD=#5
     instance {_structType=/Script/FortniteAIServer.FortStateTreeBotEquipItemTaskInstanceData, bValidateAbility=false, controller=null, itemToEquip=null}
     node {
      name=None
      _structType=/Script/FortniteAIServer.FortStateTreeBotEquipItemTask
      bCanEditConsideredForCompletion=true
      bConsideredForCompletion=true
      bHasTaskCompletionDelegateDispatcher=false
      bTaskEnabled=true
      transitionHandlingPriority=Normal
      bindingsBatch {value=65535}
      executionRuntimeTemplateIndex {value=65535}
      instanceDataHandle {index=65535, source=None, stateHandle {index=65535}}
      instanceTemplateIndex {value=65535}
      outputBindingsBatch {value=65535}
     }
    }
   }]
  transitions [
   {id=#6, trigger=ON_STATE_SUCCEEDED, state {id=#3, name=FireItem}}
   {id=#7, trigger=ON_STATE_FAILED, state {id=#8, name=Root, linkType=FAILED}}
  ]
 }
 FireItem {
  id=#3
  type=STATE
  children [#9]
  tasks [
   {
    id=#10
    description=Bot Get Current Weapon
    properties {
     iD=#11
     instance {_structType=/Script/FortniteAIServer.FortStateTreeBotGetCurrentWeaponTaskInstanceData, botController=null, currentWeapon=null}
     node {
      name=None
      _structType=/Script/FortniteAIServer.FortStateTreeBotGetCurrentWeaponTask
      bCanEditConsideredForCompletion=true
      bConsideredForCompletion=true
      bHasTaskCompletionDelegateDispatcher=false
      bTaskEnabled=true
      transitionHandlingPriority=Normal
      bindingsBatch {value=65535}
      executionRuntimeTemplateIndex {value=65535}
      instanceDataHandle {index=65535, source=None, stateHandle {index=65535}}
      instanceTemplateIndex {value=65535}
      outputBindingsBatch {value=65535}
     }
    }
   }
   {
    id=#12
    description=Bot Fire Weapon
    properties {
     iD=#13
     instance {
      _structType=/Script/FortniteAIServer.FortStateTreeBotFireWeaponTaskInstanceData
      behavior=FireSingleAndRun
      botController=null
      firingTimes []
     }
     node {
      name=None
      _structType=/Script/FortniteAIServer.FortStateTreeBotFireWeaponTask
      bCanEditConsideredForCompletion=true
      bConsideredForCompletion=true
      bHasTaskCompletionDelegateDispatcher=false
      bTaskEnabled=true
      transitionHandlingPriority=Normal
      bindingsBatch {value=65535}
      executionRuntimeTemplateIndex {value=65535}
      instanceDataHandle {index=65535, source=None, stateHandle {index=65535}}
      instanceTemplateIndex {value=65535}
      outputBindingsBatch {value=65535}
     }
    }
   }
  ]
  transitions []
 }
 WaitUntilAbilityFinished {
  id=#9
  type=STATE
  children [#14, #15]
  tasks [{
    id=#16
    description=Wait Until Weapon Ability Finished
    properties {
     iD=#17
     instance {_structType=/Script/FortniteAIServer.FortStateTreeWaitUntilAbilityFinishedTaskInstanceData, pawn=null, weapon=null}
     node {
      name=None
      _structType=/Script/FortniteAIServer.FortStateTreeWaitUntilAbilityFinishedTask
      bCanEditConsideredForCompletion=true
      bConsideredForCompletion=true
      bHasTaskCompletionDelegateDispatcher=false
      bTaskEnabled=true
      transitionHandlingPriority=Normal
      bindingsBatch {value=65535}
      executionRuntimeTemplateIndex {value=65535}
      instanceDataHandle {index=65535, source=None, stateHandle {index=65535}}
      instanceTemplateIndex {value=65535}
      outputBindingsBatch {value=65535}
     }
    }
   }]
  transitions [
   {id=#18, trigger=ON_STATE_SUCCEEDED, state {id=#8, name=Root, linkType=SUCCEEDED}}
   {id=#19, trigger=ON_STATE_FAILED, state {id=#8, name=Root, linkType=FAILED}}
  ]
 }
 WaitUntilMantisBranchAvailable {
  id=#14
  type=STATE
  tasks [{
    id=#20
    description=Wait Until Mantis Branch Available
    properties {
     iD=#21
     instance {_structType=/Script/FortniteAIServer.FortStateTreeWaitUntilMantisBranchAvailableTaskInstanceData, weapon=null}
     node {
      name=None
      _structType=/Script/FortniteAIServer.FortStateTreeWaitUntilMantisBranchAvailableTask
      bCanEditConsideredForCompletion=true
      bConsideredForCompletion=true
      bHasTaskCompletionDelegateDispatcher=false
      bTaskEnabled=true
      transitionHandlingPriority=Normal
      bindingsBatch {value=65535}
      executionRuntimeTemplateIndex {value=65535}
      instanceDataHandle {index=65535, source=None, stateHandle {index=65535}}
      instanceTemplateIndex {value=65535}
      outputBindingsBatch {value=65535}
     }
    }
   }]
  transitions [
   {id=#22, trigger=ON_STATE_SUCCEEDED, state {id=#3, name=FireItem}}
   {id=#23, trigger=ON_STATE_FAILED, state {id=#15, name=NotMantis}}
  ]
 }
 NotMantis {id=#15, type=STATE, tasks [], transitions []}
]
